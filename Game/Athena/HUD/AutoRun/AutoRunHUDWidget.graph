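# Athena HUD widget that shows the auto-run icon while the owning player has auto-run enabled.
# Construct binds controller and HUD context events that toggle the image, hide the widget on death or creative quickbar equip, and restore its saved visibility on spawn.
# Creative mode shifts the widget right for non-touch players via render transform.
# Destruct unbinds creative quickbar events but binds auto-run again and leaves death/spawn bindings in place.
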WidgetBlueprint /Game/Athena/HUD/AutoRun/AutoRunHUDWidget
graphs [
 EventGraph {nodes [
   n1 = OnAutoRunEnabled_Bind!() -> (OutputDelegate, then, bAutoRunEnabled:bool=~) {is=title K2Node_CustomEvent}
   > SetVisibility(
    self=Get Image_Run() -> (Image_Run:Image Object Reference)
    InVisibility=Select(Option 0=Hidden, Option 1=SelfHitTestInvisible, Index=@n1.bAutoRunEnabled)
   )
   OnPlayerDied!() -> (OutputDelegate, then, DeathReport:"Fort Player Death Report Structure (by ref)") {is=title K2Node_CustomEvent}
   > SetVisibility(InVisibility=Collapsed)
   OnPlayerSpawned() {is=title K2Node_CustomEvent}
   > SetVisibility(InVisibility=Get Widget_Visibility_On_Construction() -> (Widget_Visibility_On_Construction:ESlateVisibility Enum=Visible))
   Event Construct()
   > n2 = Cast To FortPlayerControllerAthena(Object=GetOwningPlayer())
   > Bind Event to On Auto Run Enabled(self:Fort Player Controller Object Reference=@n2, Delegate=Create Event()) {is=title K2Node_AddDelegate}
   > Assign On Creative Mode Enabled Changed!(
    execute:Exec
    self:Creative HUDLContext Object Reference=Get CreativeHUDLContext!() -> (ReturnValue:Creative HUDLContext Object Reference) {is=title K2Node_GetBlueprintContextEx}
    Delegate:"Delegate (by ref)"=Create Event()
   ) -> (then:Exec) {is=title K2Node_AssignDelegate}
   > Bind Event to On Creative Quickbar Equipped Changed(self:Creative HUDLContext Object Reference=Get CreativeHUDLContext(), Delegate=Create Event()) {is=title K2Node_AddDelegate}
   > Bind Event to On Local Player Died(
    self:Fort HUDContext Object Reference=Get FortHUDContext!() -> (ReturnValue:Fort HUDContext Object Reference) {is=title K2Node_GetBlueprintContextEx}
    Delegate=Create Event()
   ) {is=title K2Node_AddDelegate}
   > Set Widget_Visibility_On_Construction(Widget_Visibility_On_Construction:ESlateVisibility Enum=GetVisibility!(self:Widget Object Reference) -> (ReturnValue:ESlateVisibility Enum=Visible)) -> (Output_Get:ESlateVisibility Enum=Visible)
   > Bind Event to On Local Player Spawned(self:Fort HUDContext Object Reference=Get FortHUDContext(), Delegate=Create Event()) {is=title K2Node_AddDelegate}
   Event Destruct()
   > n3 = Cast To FortPlayerControllerAthena(Object=GetOwningPlayer())
   > Bind Event to On Auto Run Enabled(self=@n3, Delegate=Create Event())
   n4 = Get CreativeHUDLContext()
   > Is Valid(InputObject=@n4)
   > Unbind Event from On Creative Quickbar Enabled Changed(self:Creative HUDLContext Object Reference=@n4, Delegate=Create Event()) {is=title K2Node_RemoveDelegate}
   n5 = Get CreativeHUDLContext()
   > Is Valid(InputObject=@n5)
   > Unbind Event from On Creative Quickbar Equipped Changed(self:Creative HUDLContext Object Reference=@n5, Delegate=Create Event()) {is=title K2Node_RemoveDelegate}
  ]}
 OnCreativeModeEnabledChanged {nodes [
   n1 = OnCreativeModeEnabledChanged!() -> (then, bIsEnabled:bool=~) {is=title K2Node_FunctionEntry}
   n2 = GetOwningPlayer()
   n3 = AND Boolean(A=@n1, B=NOT Boolean(A=IsUsingTouch(PlayerController=@n2)))
   > SetRenderTransform!(
    execute
    self:Widget Object Reference
    InTransform:Widget Transform Structure=Make Widget Transform!(
     Translation:Vector 2D Structure=Select(Option 0:Vector 2D Structure, Option 1:Vector 2D Structure=(X=80,Y=0), Index=@n3) -> (ReturnValue:Vector 2D Structure)
     Scale:Vector 2D Structure=(X=1,Y=1)
     Shear:Vector 2D Structure=(X=0,Y=0)
     Angle:float=0
    ) -> (WidgetTransform:Widget Transform Structure) {is=title K2Node_MakeStruct}
   ) -> (then)
  ]}
 OnCreativeQuickbarEquippedChanged {nodes [
   n1 = OnCreativeQuickbarEquippedChanged!() -> (then, bIsEnabled:bool=~) {is=title K2Node_FunctionEntry}
   > n2 = Branch(Condition=@n1)
   > SetVisibility(InVisibility=Collapsed)
   SetVisibility(execute=@n2.else, InVisibility=SelfHitTestInvisible)
  ]}
]
variables [Widget_Visibility_On_Construction {type=ESlateVisibility, default=SelfHitTestInvisible}]
widgets {
 info {inheritedWidgetCount=1, parentClass=&FortHUDElementWidget, rootWidgetClass=&Image, widgetCount=2}
 widgets [
  Image_Run {
   widgetClassPath=&Image
   bIsVariable=true
   widget=Image_Run
   properties {
    visibility=Collapsed
    brush {
     bIsDynamicallyLoaded=false
     drawAs=Image
     imageType=NoImage
     mirroring=NoMirror
     resourceName=None
     resourceObject=&/M_AutoRun
     tiling=NoTile
     imageSize=<64 64>
     margin=<ltrb 0 0 0 0>
     outlineSettings {
      bUseBrushTransparency=false
      roundingType=HalfHeightRadius
      width=0
      color {colorUseRule=UseColor_Specified, specifiedColor=<rgba 0 0 0 0>}
      cornerRadii=<0 0 0 0>
     }
     tintColor {colorUseRule=UseColor_Specified, specifiedColor=<rgba 1 1 1 1>}
     uVRegion {bIsValid=false, max=<0 0>, min=<0 0>}
    }
   }
  }
  TutorialHighlight {widgetClassPath=&TutorialHighlight}
 ]
}
